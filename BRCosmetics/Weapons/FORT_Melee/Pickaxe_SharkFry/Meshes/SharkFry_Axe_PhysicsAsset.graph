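# Physics asset for the SharkFry pickaxe, a two-handed melee weapon cosmetic mesh.
# It holds a single capsule body on the Melee_TwoHanded_Handle bone, offset and tilted along the weapon.
# With no other bodies and no constraints, the pickaxe collides as one rigid shape.
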PhysicsAsset /BRCosmetics/Weapons/FORT_Melee/Pickaxe_SharkFry/Meshes/SharkFry_Axe_PhysicsAsset
bodies {Melee_TwoHanded_Handle [{shapeType=Capsule, length=86.0578, radius=28.974, center=<0.8067 2.1226 31.2557>, rotation=<rot -3.599 -90 0.0309>}]}
constraints []
